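# Material function adding view-shifting holographic foil to masked areas of an input material, built for the HighTower Squash character body.
# Mask channels define three foil layers that override metallic, roughness and base color, with fresnel toon and ink lines darkening the color.
# Foil hue samples a curve atlas by pre-skinned position, swirl-noise distortion, camera direction and layer offsets.
# Shading-path and quality switches simplify layer offsets, strengths and roughness for mobile and forward rendering.
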MaterialFunction /Game/Characters/Player/Female/Medium/Bodies/F_MED_HighTower_Squash/Materials/MF_HoloFoil
graph {nodes [
  n1 = GetMaterialAttributes(None=FunctionInput() {inputName=Material, inputType=FunctionInput_MaterialAttributes}) {attributeGetTypes [
    #1
    #2
    #3
    #4
    #5
   ]}
  n2 = LinearInterpolate(
   A=ScalarParameter() {defaultValue=1, group=HoloFoil, parameterName=HF_Metalness_Min}
   B=ScalarParameter() {defaultValue=1, group=HoloFoil, parameterName=HF_Metalness_Max}
   Alpha=@n1.Metallic
  )
  n3 = TextureSampleParameter2D() {parameterName=SuperLevel_Mask}
  n4 = ChannelMaskParameter(None=@n3.RGBA) {group=HoloFoil, parameterName=HF_Layer_01_Channel}
  n5 = ChannelMaskParameter(None=@n3.RGBA) {group=HoloFoil, maskChannel=Green, parameterName=HF_Layer_02_Channel, defaultValue=<rgba 0 1 0 0>}
  n6 = ChannelMaskParameter(None=@n3.RGBA) {group=HoloFoil, maskChannel=Blue, parameterName=HF_Layer_03_Channel, defaultValue=<rgba 0 0 1 0>}
  n7 = Clamp(None=Add(A=Add(A=@n4.R, B=@n5.R), B=@n6.R))
  n8 = ScalarParameter() {defaultValue=0.1, group=HoloFoil, parameterName=HF_Roughness_Min}
  n9 = ScalarParameter() {defaultValue=0.2, group=HoloFoil, parameterName=HF_Roughness_Max}
  n10 = ShadingPathSwitch(Default=QualitySwitch(Default=LinearInterpolate(A=@n8, B=@n9, Alpha=@n1.Roughness)), Forward=@n9, Mobile=@n9)
  n11 = Transform(None=@n1.Normal)
  n12 = Fresnel(ExponentIn=ScalarParameter() {defaultValue=6, group=HoloFoil, parameterName=BL_OutlineExponent1}, Normal=@n11) {baseReflectFraction=0}
  n13 = LinearInterpolate(
   A=ScalarParameter() {defaultValue=-4, group=HoloFoil, parameterName=BL_OutlineC1}
   B=ScalarParameter() {defaultValue=12, group=HoloFoil, parameterName=BL_OutlineC2}
   Alpha=Power(Base=OneMinus(None=@n12), Exp=ScalarParameter() {defaultValue=12, group=HoloFoil, parameterName=BL_OutlineExponent2})
  )
  n14 = OneMinus(None=ChannelMaskParameter(None=@n3.RGBA) {group=HoloFoil, maskChannel=Alpha, parameterName=HF_EdgeLineChannel, defaultValue=<rgba 0 0 0 1>}.R)
  n15 = ScalarParameter() {defaultValue=2, group=HoloFoil, parameterName=HF_LineThicknessPower}
  n16 = Multiply(
   A=LinearInterpolate(B=Saturate(None=@n13), Alpha=ScalarParameter() {defaultValue=1, group=HoloFoil, parameterName=HF_ToonLine_Opacity}) {constA=1}
   B=LinearInterpolate(B=Power(Base=@n14, Exp=@n15) {constExponent=3}, Alpha=ScalarParameter() {defaultValue=1, group=HoloFoil, parameterName=HF_InkLine_Opacity}) {constA=1}
  )
  n17 = Power(Base=DotProduct(A=@n11, B=CameraVectorWS())) {constExponent=4}
  n18 = FunctionInput(Preview=VertexInterpolator(VS=PreSkinnedPosition()).PS) {bUsePreviewValueAsDefault=true, inputName=PreskinnedLocalPosition}
  n19 = Divide(
   A=@n18
   B=Subtract(
    A=PreSkinnedLocalBounds().Max
    B=VectorParameter() {group=HoloFoil, parameterName=HT_CustomBoundsAdjustment, defaultValue=<rgba 0 0 0 1>}.RGB
   )
  )
  n20 = TextureObject() {
   samplerType=SAMPLERTYPE_LinearColor
   texture=/Game/Characters/Player/Female/Medium/Bodies/F_MED_HighTower_Squash/Textures/T_Swirly_Noise_highContrast1_Characters
  }
  n21 = PreSkinnedNormal()
  n22 = MaterialFunctionCall(
   TextureObject=@n20
   TextureSize=VectorParameter() {group=HoloFoil, parameterName=HF_ChromaDistortionNoise_Tiling, defaultValue=<rgba 256 256 256 1>}.RGB
   WorldPosition=@n18
   World Space Normal=FunctionInput(Preview=VertexInterpolator(VS=@n21).PS) {bUsePreviewValueAsDefault=true, inputName=PreskinnedLocalNormal}
  ) {materialFunction=/Game/Weapons/WeaponSkins/Wraps/MaterialFunctions/WorldAlignedTexture}
  n23 = ScalarParameter() {defaultValue=0.25, group=HoloFoil, parameterName=HF_SwirlStrength}
  n24 = ComponentMask(None=Subtract(A=@n19, B=Multiply(A=@n22.XYZ Texture, B=@n23))) {b=true}
  n25 = MaterialFunctionCall() {materialFunction=&CameraDirectionVector}
  n26 = ScalarParameter() {defaultValue=0.5, group=HoloFoil, parameterName=ViewAnlgeFidelity}
  n27 = Multiply(A=DotProduct(A=Constant3Vector() {constant=<rgba 0 0 1 0>}, B=@n25.Result), B=@n26)
  n28 = Multiply(A=DotProduct(A=Constant3Vector() {constant=<rgba 1 0 0 0>}, B=@n25.Result), B=@n26)
  n29 = Add(A=@n27, B=@n28)
  n30 = Multiply(A=DotProduct(A=Constant3Vector() {constant=<rgba 0 1 0 0>}, B=@n25.Result), B=@n26)
  n31 = Add(A=ComponentMask(None=@n19) {r=true}, B=@n28)
  n32 = Add(A=ComponentMask(None=@n19) {g=true}, B=@n30)
  n33 = LinearInterpolate(
   A=Add(A=@n24, B=Add(A=@n29, B=@n30))
   B=Add(A=Add(A=@n31, B=@n32), B=Add(A=@n24, B=@n27))
   Alpha=ScalarParameter() {group=HoloFoil, parameterName=HC_Use3Vector}
  )
  n34 = ScalarParameter() {defaultValue=0.5, group=HoloFoil, parameterName=HF_Layer_02_ChromaOffset}
  n35 = ScalarParameter() {group=HoloFoil, parameterName=HF_Layer_03_ChromaOffset}
  n36 = Add(
   A=Multiply(A=ScalarParameter() {defaultValue=1, group=HoloFoil, parameterName=HF_Layer_01_ChromaOffset}, B=@n4.R)
   B=Add(A=Multiply(A=@n34, B=@n5.R), B=Multiply(A=@n35, B=@n6.R))
  )
  n37 = Multiply(A=ScalarParameter() {defaultValue=0.5, group=HoloFoil, parameterName=HF_MobileChromaOffset}, B=@n7)
  n38 = Multiply(A=ShadingPathSwitch(Default=QualitySwitch(Default=@n36), Forward=@n37, Mobile=QualitySwitch(Default=@n36, Low=@n37, Medium=@n36)), B=@n26)
  n39 = CurveAtlasRowParameter(CurveTime=Frac(None=Add(A=@n33, B=@n38))) {
   atlas=&AthenaCurveAtlas01
   curve=&/HoloFoilChroma
   defaultValue=237
   group=LUT
   parameterName=FoilChroma
  }
  n40 = LinearInterpolate(
   A=Multiply(A=@n39, B=ScalarParameter() {defaultValue=0.5, group=HoloFoil, parameterName=HF_Layer_01_Strength})
   B=Multiply(A=@n39, B=ScalarParameter() {defaultValue=0.05, group=HoloFoil, parameterName=HF_Layer_02_Strength})
   Alpha=@n5.R
  )
  n41 = Multiply(A=@n39, B=ScalarParameter() {defaultValue=0.25, group=HoloFoil, parameterName=HF_Layer_03_Strength})
  n42 = Multiply(A=@n39, B=ScalarParameter() {defaultValue=1, group=HoloFoil, parameterName=HF_MobileLayer_Strength})
  n43 = ShadingPathSwitch(Default=QualitySwitch(Default=LinearInterpolate(A=@n40, B=@n41, Alpha=@n6.R)), Forward=@n42, Mobile=@n42)
  n44 = ScalarParameter() {group=HoloFoil, parameterName=HF_Layer_00_Strength}
  n45 = Multiply(
   A=LinearInterpolate(
    A=ScalarParameter() {defaultValue=1, group=HoloFoil, parameterName=HF_FoilMinBrightness}
    B=ScalarParameter() {defaultValue=7, group=HoloFoil, parameterName=HF_FoilMaxBrightness}
    Alpha=Clamp(None=@n17)
   ) {constB=2}
   B=LinearInterpolate(A=Multiply(A=@n43, B=@n44), B=@n43, Alpha=@n7)
  )
  n46 = LinearInterpolate(
   A=@n1.BaseColor
   B=LinearInterpolate(A=@n1.BaseColor, B=Multiply(A=@n45, B=@n16), Alpha=@n7)
   Alpha=ScalarParameter() {defaultValue=1, group=HoloFoil, parameterName=HF_BaseColorContribution}
  )
  FunctionOutput(None=SetMaterialAttributes!(
    MaterialAttributes=@n1.MaterialAttributes
    Metallic=LinearInterpolate(A=@n1.Metallic, B=@n2, Alpha=@n7)
    Roughness=LinearInterpolate(A=@n1.Roughness, B=@n10, Alpha=@n7)
    BaseColor=Multiply(A=@n46, B=@n16)
   ) {attributeSetTypes [#3, #4, #5]})
 ]}
